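# Physics asset for the Bounty Hunter backpack cosmetic, giving collision bodies and joints to the pack and its Dyn_ knife and zipper bones.
# The pack body combines a capsule and two boxes, while each knife and zipper bone uses a single box.
# The zippers form a two-link chain from the pack.
# A master knife body joins to the pack with small angle settings, and both knife pieces are angularly locked to it.
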PhysicsAsset /BRCosmetics/Accessories/FORT_Backpacks/F_MED_Bounty_Hunter_Backpack/Meshes/F_MED_Bounty_Hunter_Backpack_Physics
bodies {
 Dyn_knife_a [{shapeType=Box, extentX=20.3371, extentY=1.0018, extentZ=3.1298, center=<10.23 -0.588 0.0247>}]
 Dyn_knife_b [{shapeType=Box, extentX=23.4822, extentY=1.6308, extentZ=3.5141, center=<-4.4515 -0.1041 -0.2915>}]
 Dyn_knife_mstr [{shapeType=Box, extentX=36.51, extentY=1.01, extentZ=6.01, center=<0 -0.2306 -0.0001>}]
 Dyn_zipper01 [{shapeType=Box, extentX=4.8553, extentY=1.1233, extentZ=1.4584, center=<3.4051 -0.1076 0.2467>}]
 Dyn_zipper02 [{shapeType=Box, extentX=4.0896, extentY=1.8442, extentZ=3.2064, center=<1.4182 0.0805 0.7758>, rotation=<rot -50.3809 0 0>}]
 pack [
  {shapeType=Capsule, length=14.9742, radius=6.485, center=<-2.4954 0.8019 -3.2404>, rotation=<rot -30.5507 -28.9027 25.3036>}
  {shapeType=Box, extentX=7.5, extentY=7.5, extentZ=11.5, center=<-8.7793 -3.7932 -3.933>, rotation=<rot -32.9372 -30.3104 21.5331>}
  {shapeType=Box, extentX=6.5, extentY=6, extentZ=12.5, center=<6.7676 -1.8779 4.1743>, rotation=<rot -20.4484 -1.2824 2.0749>}
 ]
}
constraints [
 {bone1Name=Dyn_zipper02, bone2Name=Dyn_zipper01, swing1LimitDegrees=45, swing2LimitDegrees=45, twistLimitDegrees=5}
 {bone1Name=Dyn_zipper01, bone2Name=pack, swing1LimitDegrees=18.8571, swing2LimitDegrees=60, twistLimitDegrees=5}
 {bone1Name=Dyn_knife_mstr, bone2Name=pack, swing1LimitDegrees=3, swing2LimitDegrees=3.5, twistLimitDegrees=5}
 {
  bone1Name=Dyn_knife_a
  bone2Name=Dyn_knife_mstr
  swing1LimitDegrees=45
  swing1Motion=Locked
  swing2LimitDegrees=45
  swing2Motion=Locked
  twistLimitDegrees=45
  twistMotion=Locked
 }
 {
  bone1Name=Dyn_knife_b
  bone2Name=Dyn_knife_mstr
  swing1LimitDegrees=45
  swing1Motion=Locked
  swing2LimitDegrees=45
  swing2Motion=Locked
  twistLimitDegrees=45
  twistMotion=Locked
 }
]
