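# Animation Blueprint for the Bounty Hunter backpack cosmetic, adding physics-driven secondary motion that adapts to the pawn's movement state.
# Cached poses chain differently tuned RigidBody simulations through crouch, moving-crouch and targeting bool blends.
# Moving crouch half-blends the zipper branch toward the input pose.
# Jumping, set by any nonzero vertical velocity, adds a low-gravity simulation with upward force and a partly layered jump pose.
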AnimBlueprint /BRCosmetics/Accessories/FORT_Backpacks/F_MED_Bounty_Hunter_Backpack/Meshes/F_MED_Bounty_Hunter_Backpack_AnimBP
target_skeleton=&/F_MED_Bounty_Hunter_Backpack_Skeleton
graphs [
 AnimGraph {nodes [
   n1 = RigidBody(ComponentPose=Local To Component(LocalPose=Use cached pose 'IfTargeting'()), ExternalForce="0,0,75") {
    bClampLinearTranslationLimitToRefPose=true
    bForceDisableCollisionBetweenConstraintBodies=true
    bOverrideWorldGravity=true
    overridePhysicsAsset=&/F_MED_Bounty_Hunter_Backpack_Physics
    simulationSpace=BaseBoneSpace
    alphaBoolBlend {blendOption=Linear}
    baseBoneRef {boneName=pack}
    componentAppliedLinearAccClamp=<3000 3000 3000>
    componentLinearAccScale=<0.1 0.1 0.1>
    componentLinearVelScale=<0.1 0.1 0.1>
    overrideWorldGravity {z=-386}
   }
   n2 = Component To Local(ComponentPose=@n1)
   n3 = Sequence Player 'F_MED_Bounty_Hunter_Backpack_JumpPose'() {sequence=&/F_MED_Bounty_Hunter_Backpack_JumpPose}
   Output Pose(Result=Blend Poses by bool(
     BlendPose_0=Layered blend per bone(BasePose=@n2, BlendPoses_0=@n3, BlendWeights_0=0.33) {layerSetup [{branchFilters [{blendDepth=2, boneName=Dyn_zipper01}]}]}
     BlendPose_1=Use cached pose 'IfTargeting'()
     bActiveValue=Get Jumping() -> (Jumping:bool)
    ) {blendType=Linear}) {layerGroup=None}
   AnimGraphNode_SubInput_0 = Input Pose()
   Save cached pose 'Subgraph'(Pose=@AnimGraphNode_SubInput_0) {cacheName=Subgraph}
   n4 = RigidBody(ComponentPose=Local To Component(LocalPose=Use cached pose 'Subgraph'())) {
    bClampLinearTranslationLimitToRefPose=true
    bOverrideWorldGravity=true
    overridePhysicsAsset=&F_MED_Bounty_Hunter_Backpack_Physics
    alphaBoolBlend {blendOption=Linear}
    componentLinearAccScale=<0.5 0.5 0.5>
    componentLinearVelScale=<0.5 0.5 0.5>
    overrideWorldGravity {z=-980}
   }
   Save cached pose 'InitDynamics'(Pose=Component To Local(ComponentPose=@n4)) {cacheName=InitDynamics}
   Save cached pose 'IfCrouching'(Pose=Blend Poses by bool(
     BlendPose_0=Use cached pose 'CrouchHack'()
     BlendPose_1=Use cached pose 'IfCrouchingSlow'()
     bActiveValue=Get SprintCrouching() -> (SprintCrouching:bool)
    ) {blendType=Linear}) {cacheName=IfCrouching}
   n5 = RigidBody(ComponentPose=Local To Component(LocalPose=Use cached pose 'IfCrouching'())) {
    bClampLinearTranslationLimitToRefPose=true
    bOverrideWorldGravity=true
    overridePhysicsAsset=&F_MED_Bounty_Hunter_Backpack_Physics
    alphaBoolBlend {blendOption=Linear}
    componentLinearAccScale=<0.05 0.05 0.05>
    componentLinearVelScale=<0.05 0.05 0.05>
    overrideWorldGravity {z=-980}
   }
   Save cached pose 'IfTargeting'(Pose=Blend Poses by bool(BlendPose_0=Component To Local(ComponentPose=@n5), BlendPose_1=Use cached pose 'IfCrouching'(), bActiveValue=Get bIsTargeting()) {blendType=Linear}) {cacheName=IfTargeting}
   n6 = RigidBody(ComponentPose=Local To Component(LocalPose=Use cached pose 'Subgraph'())) {
    bClampLinearTranslationLimitToRefPose=true
    bOverrideWorldGravity=true
    overridePhysicsAsset=&F_MED_Bounty_Hunter_Backpack_Physics
    simulationSpace=BaseBoneSpace
    alphaBoolBlend {blendOption=Linear}
    baseBoneRef {boneName=root}
    componentAppliedLinearAccClamp=<3000 3000 3000>
    componentLinearAccScale=<1 1 1>
    componentLinearVelScale=<1 1 1>
    overrideWorldGravity {z=-980}
   }
   Save cached pose 'IfCrouchingSlow'(Pose=Blend Poses by bool(BlendPose_0=Component To Local(ComponentPose=@n6), BlendPose_1=Use cached pose 'InitDynamics'(), bActiveValue=Get bIsCrouching()) {blendType=Linear}) {cacheName=IfCrouchingSlow}
   Save cached pose 'CrouchHack'(Pose=Layered blend per bone(BasePose=Use cached pose 'IfCrouchingSlow'(), BlendPoses_0=Use cached pose 'Subgraph'(), BlendWeights_0=0.5) {layerSetup [{branchFilters [{blendDepth=1, boneName=Dyn_zipper01}]}]}) {cacheName=CrouchHack}
   Use cached pose 'CrouchHack'()
  ]}
 EventGraph {nodes [
   Event BlueprintUpdateAnimation()
   > n1 = Sequence()
   > Set bUpdateOwnerAnimInputProperty(bUpdateOwnerAnimInputProperty:bool=true) -> (Output_Get:bool)
   n2 = Get PawnVelocityZ()
   > n3 = Branch(Condition="Not Equal (Float)"!(A:double=@n2, B:double=0) -> (ReturnValue:bool) {is=title K2Node_CallFunction})
   > Set Jumping(Jumping:bool=true) -> (Output_Get:bool)
   Set Jumping(execute=@n3.else, Jumping=false)
   n4 = Get PawnSpeed2D()
   n5 = Branch(execute=@n1.then_1, Condition=AND Boolean(A=Get bIsCrouching(), B="float > float"(A=@n4, B=50) {is=title K2Node_CallFunction}))
   > Set SprintCrouching(SprintCrouching:bool=true) -> (Output_Get:bool)
   Set SprintCrouching(execute=@n5.else, SprintCrouching=false)
   TryGetPawnOwner()
   GetPlayerCharacter!(PlayerIndex:int=0) -> (ReturnValue:Character Object Reference)
   "float > float"(A=@n2, B=0.5)
  ]}
]
variables [
 Jumping {type=boolean}
 SprintCrouching {type=boolean}
]
